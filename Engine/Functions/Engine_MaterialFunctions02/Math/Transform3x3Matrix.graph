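# Library material function, listed under Misc and Math, that transforms a vector into a new basis defined by three axis vectors and a position.
# It splits VectorToTransform into components, scales BasisX, BasisY and BasisZ by them, and sums the results with BasisPosition.
# BasisPosition is always added, as the vector's alpha mentioned in the input description is not read.
# Unconnected basis inputs default to identity axes and a zero offset.
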MaterialFunction /Engine/Functions/Engine_MaterialFunctions02/Math/Transform3x3Matrix
graph {
 properties {bExposeToLibrary=true, description=Transform a vector into a new basis, libraryCategoriesText [Misc,Math]}
 nodes [
  n1 = FunctionInput() {
   bUsePreviewValueAsDefault=true
   desc=transforms single vector
   description="The alpha can be a 1 if the vector contains a position and a 0 if the basis contains direction. "
   inputName=VectorToTransform
   previewValue=<0 0 1 1>
  }
  n2 = MaterialFunctionCall(Float3=@n1) {materialFunction=&BreakOutFloat3Components}
  n3 = Multiply(A=FunctionInput() {bUsePreviewValueAsDefault=true, inputName=BasisX, sortPriority=7, previewValue=<1 0 0 0>}, B=@n2.R)
  n4 = Multiply(A=FunctionInput() {bUsePreviewValueAsDefault=true, inputName=BasisY, sortPriority=8, previewValue=<0 1 0 0>}, B=@n2.G)
  n5 = Multiply(A=FunctionInput() {bUsePreviewValueAsDefault=true, inputName=BasisZ, sortPriority=9, previewValue=<0 0 1 0>}, B=@n2.B)
  n6 = FunctionInput() {bUsePreviewValueAsDefault=true, inputName=BasisPosition, sortPriority=10}
  FunctionOutput(None=Add(A=Add(A=@n3, B=@n4), B=Add(A=@n5, B=@n6))) {outputName=Transformed Vector}
 ]
}
