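# Material function that maps a 0-1 XY phase and a Z phase to a point on a box's four side walls, for evenly arranging models around it.
# The XY phase is split into four edges proportional to X and Y vector lengths, selected with floor/saturate lerp chains.
# The unit-box point is offset by a relative pivot, then transformed by the basis vectors and position.
# A second output gives each side's normalized normal.
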MaterialFunction /Engine/Functions/Engine_MaterialFunctions02/Math/ArrangePointsEvenlyAroundABox
graph {
 properties {
  bExposeToLibrary=true
  description="This material function will evenly arrange points around a box. The dimensions and basis vectors are inputs along with a 0-1 placement value. This can be useful for evenly distributing models around an analytically described shape. "
  libraryCategoriesText [Math]
 }
 nodes [
  n1 = FunctionInput(Preview=ComponentMask(None=TextureCoordinate()) {r=true}) {bUsePreviewValueAsDefault=true, inputName="XY Phase (0-1)", inputType=FunctionInput_Scalar}
  n2 = FunctionInput() {
   bUsePreviewValueAsDefault=true
   description=Define the box's x basis vector and the length along it's axis.
   inputName=X Vector
   sortPriority=-10
   previewValue=<10 0 0 1>
  }
  n3 = MaterialFunctionCall(Vector 3=@n2) {materialFunction=/Engine/Functions/Engine_MaterialFunctions02/Utility/VectorLength}
  n4 = FunctionInput() {
   bUsePreviewValueAsDefault=true
   description=Define the box's Y basis vector and the length along it's axis.
   inputName=Y Vector
   sortPriority=-9
   previewValue=<0 40 0 1>
  }
  n5 = Divide(
   A=@n3.V3 Length
   B=Add(
    A=@n3.V3 Length
    B=MaterialFunctionCall(Vector 3=@n4) {materialFunction=&VectorLength}.V3 Length
   )
  )
  n6 = Multiply(A=@n5) {constB=0.5, desc=side 1 max number}
  n7 = Divide(A=@n1, B=@n6)
  n8 = Constant3Vector() {constant=<rgba 1 0 0 0>}
  n9 = Multiply(A=@n8, B=@n7)
  n10 = Constant3Vector() {constant=<rgba 0 1 0 0>}
  n11 = Subtract(A=@n1, B=@n6)
  n12 = Multiply(A=OneMinus(None=@n5)) {constB=0.5, desc="side 2 min number "}
  n13 = Divide(A=@n11, B=@n12)
  n14 = Add(A=Multiply(A=OneMinus(None=@n13), B=@n10), B=@n8)
  n15 = Subtract(A=@n11, B=@n12)
  n16 = Divide(A=@n15, B=@n6)
  n17 = OneMinus(None=@n16)
  n18 = Constant() {desc=the grid goes from 0-1}
  n19 = Divide(A=Subtract(A=@n15, B=@n6), B=@n12)
  n20 = Saturate(None=Floor(None=@n16)) {desc=top black current}
  n21 = LinearInterpolate(A=Add(A=Multiply(A=@n17, B=@n8), B=@n18), B=Add(A=Multiply(A=@n19, B=@n10), B=@n18), Alpha=@n20) {constB=0, desc=x}
  n22 = Saturate(None=Floor(None=@n13)) {desc=top black current}
  n23 = Saturate(None=Floor(None=@n7))
  n24 = FunctionInput(Preview=ComponentMask(None=TextureCoordinate()) {g=true}) {
   bUsePreviewValueAsDefault=true
   description=This value will place the points along the defined z vector.
   inputName="Z Phase (0-1)"
   inputType=FunctionInput_Scalar
   sortPriority=1
   previewValue=<50 0 0 1>
  }
  n25 = Add(
   A=LinearInterpolate(A=Add(A=@n9, B=@n10), B=LinearInterpolate(A=@n14, B=@n21, Alpha=@n22) {constB=0, desc=y}, Alpha=@n23) {constB=0, desc=x}
   B=Multiply(A=@n24, B=Constant3Vector() {constant=<rgba 0 0 1 0>})
  )
  n26 = FunctionInput() {bUsePreviewValueAsDefault=true, inputName=Relative Pivot Location, sortPriority=-1, previewValue=<-0.5 -0.5 -0.5 1>}
  n27 = FunctionInput() {
   bUsePreviewValueAsDefault=true
   description=Define the box's z basis vector and the length along it's axis.
   inputName=Z Vector
   sortPriority=-7
   previewValue=<0 0 20 1>
  }
  FunctionOutput(None=MaterialFunctionCall(
    VectorToTransform=Add(A=@n25, B=@n26)
    BasisX=@n2
    BasisY=@n4
    BasisZ=@n27
    BasisPosition=FunctionInput() {bUsePreviewValueAsDefault=true, description=This defines the boxes Center Location, inputName=Position, sortPriority=-6}
   ) {materialFunction=&Transform3x3Matrix}.Transformed Vector) {
   description="If you wish to use this to generate a box in world space via world position offset, subtract world position from this output before running it into world position offset. "
   outputName=Point Around Box
  }
  n28 = Constant3Vector() {constant=<rgba 0 -1 0 0>}
  n29 = Constant3Vector() {constant=<rgba -1 0 0 0>}
  n30 = LinearInterpolate(
   A=Constant3Vector() {constant=<rgba 0 1 0 0>}
   B=LinearInterpolate(A=Constant3Vector() {constant=<rgba 1 0 0 0>}, B=LinearInterpolate(A=@n28, B=@n29, Alpha=@n20) {constB=0, desc=x}, Alpha=@n22) {constB=0, desc=y}
   Alpha=@n23
  ) {constB=0, desc=x}
  n31 = Constant3Vector()
  FunctionOutput(None=Normalize(VectorInput=MaterialFunctionCall(VectorToTransform=@n30, BasisX=@n2, BasisY=@n4, BasisZ=@n31, BasisPosition=@n31) {materialFunction=&Transform3x3Matrix}.Transformed Vector)) {description=This is the world normal for the outward facing surface., outputName=World normal, sortPriority=1}
 ]
}
