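# Material function that concatenates two 4x4 matrices, each supplied as four float4 basis inputs that default to their preview values.
# It breaks the second matrix's vectors apart and regroups them by component, then dots each first-matrix vector with each group.
# Each output X, Y, Z, W gathers one component group's dot products, so the product comes out transposed relative to the input layout.
# An unconnected static bool input named with dashes separates the two input sets.
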MaterialFunction /Engine/Functions/Engine_MaterialFunctions02/Math/ConcatenateMatrices
graph {
 properties {description=Concatenate 2 matrices, libraryCategoriesText [Misc,Math]}
 nodes [
  FunctionInput() {bUsePreviewValueAsDefault=true, inputName=----------, inputType=FunctionInput_StaticBool, sortPriority=5}
  n1 = FunctionInput() {bUsePreviewValueAsDefault=true, inputName=BasisX, inputType=FunctionInput_Vector4, sortPriority=1, previewValue=<1 0 0 0>}
  n2 = FunctionInput() {bUsePreviewValueAsDefault=true, inputName=BasisX2, inputType=FunctionInput_Vector4, sortPriority=7, previewValue=<1 0 0 0>}
  n3 = MaterialFunctionCall(Float4=@n2) {materialFunction=&BreakOutFloat4Components}
  n4 = FunctionInput() {bUsePreviewValueAsDefault=true, inputName=BasisY2, inputType=FunctionInput_Vector4, sortPriority=8, previewValue=<0 1 0 0>}
  n5 = MaterialFunctionCall(Float4=@n4) {materialFunction=&BreakOutFloat4Components}
  n6 = FunctionInput() {bUsePreviewValueAsDefault=true, inputName=BasisZ2, inputType=FunctionInput_Vector4, sortPriority=9, previewValue=<0 0 1 0>}
  n7 = MaterialFunctionCall(Float4=@n6) {materialFunction=&BreakOutFloat4Components}
  n8 = MaterialFunctionCall(Float4=FunctionInput() {bUsePreviewValueAsDefault=true, inputName=BasisW2, inputType=FunctionInput_Vector4, sortPriority=10}) {materialFunction=&BreakOutFloat4Components}
  n9 = MaterialFunctionCall(X=@n3.B, Y=@n5.B, Z=@n7.B, A=@n8.B) {materialFunction=&MakeFloat4}
  n10 = FunctionInput() {bUsePreviewValueAsDefault=true, inputName=BasisY, inputType=FunctionInput_Vector4, sortPriority=2, previewValue=<0 1 0 0>}
  n11 = AppendVector(A=DotProduct(A=@n1, B=@n9.Result), B=DotProduct(A=@n10, B=@n9.Result))
  n12 = FunctionInput() {bUsePreviewValueAsDefault=true, inputName=BasisZ, inputType=FunctionInput_Vector4, sortPriority=3, previewValue=<0 0 1 0>}
  n13 = DotProduct(A=@n12, B=@n9.Result)
  n14 = FunctionInput() {bUsePreviewValueAsDefault=true, inputName=BasisW, inputType=FunctionInput_Vector4, sortPriority=4}
  FunctionOutput(None=AppendVector(A=AppendVector(A=@n11, B=@n13), B=DotProduct(A=@n14, B=@n9.Result))) {outputName=Z, sortPriority=2}
  n15 = MaterialFunctionCall(X=@n3.A, Y=@n5.A, Z=@n7.A, A=@n8.A) {materialFunction=&MakeFloat4}
  n16 = AppendVector(A=DotProduct(A=@n1, B=@n15.Result), B=DotProduct(A=@n10, B=@n15.Result))
  n17 = DotProduct(A=@n12, B=@n15.Result)
  FunctionOutput(None=AppendVector(A=AppendVector(A=@n16, B=@n17), B=DotProduct(A=@n14, B=@n15.Result))) {outputName=W, sortPriority=3}
  n18 = MaterialFunctionCall(X=@n3.R, Y=@n5.R, Z=@n7.R, A=@n8.R) {materialFunction=&MakeFloat4}
  n19 = AppendVector(A=DotProduct(A=@n1, B=@n18.Result), B=DotProduct(A=@n10, B=@n18.Result))
  n20 = DotProduct(A=@n12, B=@n18.Result)
  FunctionOutput(None=AppendVector(A=AppendVector(A=@n19, B=@n20), B=DotProduct(A=@n14, B=@n18.Result))) {outputName=X}
  n21 = MaterialFunctionCall(X=@n3.G, Y=@n5.G, Z=@n7.G, A=@n8.G) {materialFunction=&MakeFloat4}
  n22 = AppendVector(A=DotProduct(A=@n1, B=@n21.Result), B=DotProduct(A=@n10, B=@n21.Result))
  n23 = DotProduct(A=@n12, B=@n21.Result)
  FunctionOutput(None=AppendVector(A=AppendVector(A=@n22, B=@n23), B=DotProduct(A=@n14, B=@n21.Result))) {outputName=Y, sortPriority=1}
 ]
}
