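# Engine material function that corrects a Pixel Depth Offset amount for view angle, extending the offset distance as the camera aligns with the surface plane.
# It divides the input scalar by the dot product of the camera vector and the world-space vertex normal, which is multiplied by the two-sided sign.
# The camera vector comes from CameraVectorWithWPOOptions, using its World Position Offsets variant.
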MaterialFunction /Engine/Functions/Engine_MaterialFunctions02/Math/CorrectPixelDepthOffsetValue
graph {
 properties {description=Extends pdo distance as the camera aligns with the surface plane.}
 nodes [
  n1 = MaterialFunctionCall!(Use World Position Offsets) {materialFunction=/Engine/Functions/Engine_MaterialFunctions02/Utility/CameraVectorWithWPOOptions}
  n2 = Multiply(A=VertexNormalWS(), B=TwoSidedSign())
  FunctionOutput(None=Divide(A=FunctionInput() {inputName=Pixel Depth Offset Amount, inputType=FunctionInput_Scalar}, B=DotProduct(A=@n1.Result, B=@n2))) {outputName=Angle Correct Depth Value}
 ]
}
