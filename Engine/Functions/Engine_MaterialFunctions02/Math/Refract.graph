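# Engine material function that computes a refracted vector from Normal, IncidenceVector and a scalar RatioOfIndicesOfRefraction.
# It derives the incident cosine by dotting the normal with the negated incidence vector, then the squared transmitted cosine via Snell's law.
# The Snell vector uses an absolute value inside the square root.
# The result is multiplied by the raw squared cosine rather than a step, so it is not normalized and flips sign under total internal reflection.
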MaterialFunction /Engine/Functions/Engine_MaterialFunctions02/Math/Refract
graph {nodes [
  n1 = FunctionInput() {inputName=Normal}
  n2 = FunctionInput() {inputName=IncidenceVector}
  n3 = DotProduct(A=@n1, B=Multiply(A=@n2, B=Constant3Vector() {constant=<rgba -1 -1 -1 0>})) {desc=cosi}
  n4 = FunctionInput() {inputName=RatioOfIndicesOfRefraction, inputType=FunctionInput_Scalar}
  n5 = Multiply(A=@n3, B=@n3)
  n6 = OneMinus(None=Multiply(A=OneMinus(None=@n5), B=Multiply(A=@n4, B=@n4))) {desc=cost2}
  n7 = Abs(None=@n6)
  n8 = Multiply(A=@n1, B=Subtract(A=Multiply(A=@n3, B=@n4), B=SquareRoot(None=@n7)))
  n9 = Multiply(A=@n2, B=@n4)
  FunctionOutput(None=Multiply(A=Add(A=@n8, B=@n9) {desc=t}, B=@n6))
 ]}
