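# Library material function that builds three mutually orthogonal unit vectors from two input vectors, favoring the first one's direction.
# It normalizes both inputs, passes the first through as Vector1, and crosses it with the second to form Vector3.
# Crossing that result with Vector1 gives a corrected Vector2.
# Both derived vectors are output through SafeNormalize, while two plain Normalize nodes on the same results are left unconnected.
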MaterialFunction /Engine/Functions/Engine_MaterialFunctions02/Math/CreateThirdOrthogonalVector
graph {
 properties {
  bExposeToLibrary=true
  description="Takes two vectors and generates a third orthogonal vector from them. It also ensures that all of the vectors are orthogonal to each other.  The first vectors directon is favored over the second."
  libraryCategoriesText [Misc,Math]
 }
 nodes [
  n1 = Normalize(VectorInput=FunctionInput() {inputName=Vector1, previewValue=<1 0 0 1>})
  FunctionOutput(None=@n1) {outputName=Vector1}
  n2 = CrossProduct(A=@n1, B=Normalize(VectorInput=FunctionInput() {inputName=Vector2, sortPriority=1, previewValue=<0 0 1 1>}))
  n3 = CrossProduct(A=@n2, B=@n1)
  FunctionOutput(None=MaterialFunctionCall(Vector=@n3) {materialFunction=&SafeNormalize}.Result) {outputName=Vector2, sortPriority=1}
  FunctionOutput(None=MaterialFunctionCall(Vector=@n2) {materialFunction=&SafeNormalize}.Result) {outputName=Vector3, sortPriority=2}
  Normalize(VectorInput=@n3)
  Normalize(VectorInput=@n2)
 ]
}
