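# Library material function that transforms a tangent-space vector using mesh vertex normals instead of tangent vectors, intended for world-aligned textures.
# CreateThirdOrthogonalVector takes the normalized vertex normal, scaled by the "Squash Vertex Normals" input, plus a constant Y axis, and feeds the basis of Transform3x3Matrix.
# Before that transform, the input vector's X is negated where the vertex normal has positive Z, using Ceil of a dot product as the selector.
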MaterialFunction /Engine/Functions/Engine_MaterialFunctions02/Math/TransformNormals_Tangent_to_Vertex
graph {
 properties {
  bExposeToLibrary=true
  description=Useful for when you want a World Aligned texture to pick up on mesh vertex normals but NOT tangent vectors.
 }
 nodes [
  n1 = VertexNormalWS()
  n2 = Constant3Vector() {constant=<rgba 0 0 1 0>}
  n3 = LinearInterpolate(Alpha=Ceil(None=DotProduct(A=@n1, B=@n2) {desc=generate sign for this projection})) {constA=1, constAlpha=0, constB=-1}
  n4 = Constant() {r=1}
  n5 = Constant() {r=1}
  n6 = VertexNormalWS()
  n7 = FunctionInput() {bUsePreviewValueAsDefault=true, inputName=Squash Vertex Normals, sortPriority=1, previewValue=<1 1 1 1>}
  n8 = MaterialFunctionCall!(Vector1=Normalize(VectorInput=Multiply(A=@n6, B=@n7)), Vector2=Constant3Vector() {constant=<rgba 0 1 0 0>}) {materialFunction=&/CreateThirdOrthogonalVector}
  n9 = MaterialFunctionCall(
   VectorToTransform=Multiply(
    A=FunctionInput() {bUsePreviewValueAsDefault=true, inputName=Vector to Transform, previewValue=<0 0 1 1>}
    B=MaterialFunctionCall(X=@n3, Y=@n4, Z=@n5) {materialFunction=&MakeFloat3}.Result
   )
   BasisX=@n8.Vector3
   BasisY=@n8.Vector3
   BasisZ=@n8.Vector3
  ) {
   desc=lerps in when the vertex normal faces X or Negative X
   materialFunction=&Transform3x3Matrix
  }
  FunctionOutput(None=@n9.Transformed Vector)
 ]
}
