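# Material function that projects a world position into screen-relative clip-space XY, exposed in the Misc and Math library categories.
# The input defaults to World Position excluding all material shader offsets.
# A Custom HLSL node multiplies it by the view's WorldToClip matrix, cast from Large World Coordinates to float.
# The output divides XY by W for normalized device coordinates, with no remap to 0–1 UVs despite the description mentioning UV space.
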MaterialFunction /Engine/Functions/Engine_MaterialFunctions02/Math/TransformToClipSpace
graph {
 properties {description="Transforms a world position into clip space (render buffer uv space).", libraryCategoriesText [Misc,Math]}
 nodes [
  MaterialExpressionFunctionInput_0 = FunctionInput(Preview=WorldPosition() {worldPositionShaderOffset=WPT_ExcludeAllShaderOffsets}.XYZ) {bUsePreviewValueAsDefault=true, inputName=World Position}
  n1 = Custom!(In=@MaterialExpressionFunctionInput_0) {
   code="mul(float4(In.xyz,1),LWCToFloat(ResolvedView.WorldToClip))"
   desc="\n*** Original source before expression upgrade ***\nmul(float4(In.xyz,1),View.WorldToClip)"
   outputType=CMOT_Float4
   inputs [{inputName=In, input {expression=&self:MaterialExpressionFunctionInput_0}}]
  }
  FunctionOutput(None=Divide(A=ComponentMask(None=@n1) {g=true, r=true}, B=ComponentMask(None=@n1) {a=true}))
 ]
}
