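# Material function that sums the components of a float2, float3 or float4, exposed to the library under Misc and Math.
# It has three independent input and output pairs, f2, f3 and f4, one per vector width.
# Each path splits its input with single-channel component masks and chains Add nodes to produce a scalar total.
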MaterialFunction /Engine/Functions/Engine_MaterialFunctions02/Math/AddComponents
graph {
 properties {
  bExposeToLibrary=true
  description="Input a float 2, 3 or 4 and retrieve the sum of the component values."
  libraryCategoriesText [Misc,Math]
 }
 nodes [
  n1 = FunctionInput() {inputName=f2, inputType=FunctionInput_Vector2}
  FunctionOutput(None=Add(A=ComponentMask(None=@n1) {r=true}, B=ComponentMask(None=@n1) {g=true})) {outputName=f2}
  n2 = FunctionInput() {inputName=f3, sortPriority=1}
  n3 = ComponentMask(None=@n2) {r=true}
  n4 = ComponentMask(None=@n2) {g=true}
  FunctionOutput(None=Add(A=Add(A=@n3, B=@n4), B=ComponentMask(None=@n2) {b=true})) {outputName=f3, sortPriority=1}
  n5 = FunctionInput() {inputName=f4, inputType=FunctionInput_Vector4, sortPriority=2}
  n6 = Add(A=ComponentMask(None=@n5) {r=true}, B=ComponentMask(None=@n5) {g=true})
  n7 = ComponentMask(None=@n5) {b=true}
  FunctionOutput(None=Add(A=Add(A=@n6, B=@n7), B=ComponentMask(None=@n5) {a=true})) {outputName=f4, sortPriority=2}
 ]
}
