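# Material function in the Math library that finds the angle between two normalized vectors.
# It feeds their dot product into a fast approximate arccosine, then outputs the result in degrees and as a 0-1 value scaled from 180 degrees.
# Inputs are not normalized internally; only their preview defaults, a centered-UV vector and the X axis, use SafeNormalize.
# A 90-degree color comparison and a radians-to-0-1 multiply feed no output.
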MaterialFunction /Engine/Functions/Engine_MaterialFunctions02/Math/AngleBetweenVectors
graph {
 properties {bExposeToLibrary=true, description=Find the angle between two normalized vectors in degrees., libraryCategoriesText [Math]}
 nodes [
  n1 = TextureCoordinate()
  n2 = MaterialFunctionCall(Vector=AppendVector(A=Subtract(A=@n1) {constB=0.5}, B=Constant())) {materialFunction=&SafeNormalize}
  n3 = MaterialFunctionCall(Vector=Constant3Vector() {constant=<rgba 1 0 0 0>}) {materialFunction=&SafeNormalize}
  n4 = ArccosineFast!(None=DotProduct(
    A=FunctionInput(Preview=@n2.Result) {bUsePreviewValueAsDefault=true, inputName=Normalized V1}
    B=FunctionInput(Preview=@n3.Result) {bUsePreviewValueAsDefault=true, inputName=Normalized V2, sortPriority=1}
   ))
  n5 = Multiply(A=@n4) {constB=57.2958, desc=Output in degrees}
  FunctionOutput(None=@n5) {outputName=Angle In Degrees}
  FunctionOutput(None=Multiply(A=@n5) {constB=0.0056}) {outputName=0-1, sortPriority=1}
  If(
   A=@n5
   B=Constant() {r=90}
   "A > B"=Constant3Vector() {constant=<rgba 1 0 0 0>}
   "A == B"=Constant3Vector() {constant=<rgba 0 1 0 0>}
   "A < B"=Constant3Vector() {constant=<rgba 0 0 1 0>}
  )
  Multiply(A=@n4) {constB=0.3183, desc=0-1}
 ]
}
