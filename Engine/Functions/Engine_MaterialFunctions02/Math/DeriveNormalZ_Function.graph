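# Material function in the Misc and Math library that rebuilds a normalized 3D vector from its XY components, extending the DeriveNormalZ node.
# DeriveNormalZ computes Z from the XY input, then a Z sign input, defaulting to positive, multiplies it to allow negative Z.
# XY and signed Z are appended and passed through SafeNormalize, which falls back to a configurable default vector when normalization fails, such as at zero length.
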MaterialFunction /Engine/Functions/Engine_MaterialFunctions02/Math/DeriveNormalZ_Function
graph {
 properties {
  bExposeToLibrary=true
  description="DeriveNormalZ expanded into a material function, including helpful features like safe normalization and a z sign input."
  libraryCategoriesText [Misc,Math]
 }
 nodes [
  n1 = Constant3Vector() {constant=<rgba 0.6035 0.5646 0.5628 0>}
  n2 = FunctionInput(Preview=ComponentMask(None=Normalize(VectorInput=@n1)) {g=true, r=true}) {inputName=XY vector, inputType=FunctionInput_Vector2, previewValue=<1 0 0 1>}
  n3 = FunctionInput() {
   bUsePreviewValueAsDefault=true
   description=Defaults to 1.0. Alternatively enter -1.0.
   inputName=Z sign
   inputType=FunctionInput_Scalar
   sortPriority=1
   previewValue=<1 0 0 1>
  }
  n4 = Multiply(A=ComponentMask(None=DeriveNormalZ(InXY=@n2) {desc="sqrt(saturate(1 - x^2 - y^2))"}) {b=true}, B=@n3)
  n5 = FunctionInput() {
   bUsePreviewValueAsDefault=true
   description=If the function fails this value will be used instead. It could fail if the vectors length is 0.
   inputName=Default Normalization vector
   sortPriority=2
   previewValue=<0 0 1 1>
  }
  FunctionOutput(None=MaterialFunctionCall(Vector=AppendVector(A=@n2, B=@n4), Default=@n5) {materialFunction=&SafeNormalize}.Result)
 ]
}
