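# InverseTransformMatrix is a library material function under Math that transforms a vector into a new basis given by BasisX, BasisY and BasisZ inputs.
# It takes the dot product of VectorToTransform with each basis vector and appends the three results into the Transformed Vector 3 output.
# This amounts to multiplying by the transposed basis matrix, a true inverse only for an orthonormal basis.
# The basis inputs default to the identity axes.
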MaterialFunction /Engine/Functions/Engine_MaterialFunctions02/Math/InverseTransformMatrix
graph {
 properties {bExposeToLibrary=true, description=Transform a vector into a new basis, libraryCategoriesText [Misc,Math]}
 nodes [
  n1 = FunctionInput() {bUsePreviewValueAsDefault=true, desc=transforms single vector, inputName=VectorToTransform, previewValue=<0 0 1 1>}
  n2 = DotProduct(A=FunctionInput() {bUsePreviewValueAsDefault=true, inputName=BasisX, sortPriority=7, previewValue=<1 0 0 0>}, B=@n1)
  n3 = DotProduct(A=FunctionInput() {bUsePreviewValueAsDefault=true, inputName=BasisY, sortPriority=8, previewValue=<0 1 0 0>}, B=@n1)
  n4 = FunctionInput() {bUsePreviewValueAsDefault=true, inputName=BasisZ, sortPriority=9, previewValue=<0 0 1 0>}
  FunctionOutput(None=AppendVector(A=AppendVector(A=@n2, B=@n3), B=DotProduct(A=@n4, B=@n1))) {outputName=Transformed Vector 3}
 ]
}
