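# Material function in the Math library category that takes three input vectors and outputs three mutually orthogonal vectors.
# It normalizes the inputs, then crosses Vector3×Vector1 with Vector1×Vector2 to form the first output.
# The second output crosses Vector1×Vector2 with the first, and the third crosses the first two outputs.
# Only the inputs are normalized, so the outputs are not guaranteed unit length.
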MaterialFunction /Engine/Functions/Engine_MaterialFunctions02/Math/MakeVectorsOrthogonal
graph {
 properties {bExposeToLibrary=true, description=Makes vectors orthagonal by using the cross product of the vectors., libraryCategoriesText [Misc,Math]}
 nodes [
  n1 = FunctionInput() {inputName=Vector3, sortPriority=2, previewValue=<0 0 1 1>}
  n2 = Normalize(VectorInput=FunctionInput() {inputName=Vector1, previewValue=<1 0 0 1>})
  n3 = CrossProduct(A=@n2, B=Normalize(VectorInput=FunctionInput() {inputName=Vector2, sortPriority=1, previewValue=<0 1 0 1>}))
  n4 = CrossProduct(A=CrossProduct(A=Normalize(VectorInput=@n1), B=@n2), B=@n3)
  FunctionOutput(None=@n4) {outputName=Vector1}
  n5 = CrossProduct(A=@n3, B=@n4)
  FunctionOutput(None=@n5) {outputName=Vector2, sortPriority=1}
  FunctionOutput(None=CrossProduct(A=@n4, B=@n5) {desc=You've crossed me for the last time...}) {outputName=Vector3, sortPriority=2}
 ]
}
