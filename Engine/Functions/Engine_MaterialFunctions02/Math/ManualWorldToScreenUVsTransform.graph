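# Material function that manually projects a camera-relative position into screen UVs for any perspective viewport, useful in render-to-texture passes.
# Dot products with three camera axis inputs give local coordinates; X and Y are divided by depth times tan(FOV/2), then remapped to UVs with Y flipped.
# Inputs default to the current view's values, but overrides remove that dependency.
# An unconnected node chain illustrates computing the FOV input from FOV and screen resolution.
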MaterialFunction /Engine/Functions/Engine_MaterialFunctions02/Math/ManualWorldToScreenUVsTransform
graph {
 properties {
  bExposeToLibrary=true
  description=This function will allow one to manually recreate any perspective viewport without the use of any intrinsic view properties. This is helpful for Render to texture passes.
  libraryCategoriesText [WPO]
 }
 nodes [
  n1 = FunctionInput(Preview=Subtract(A=WorldPosition().XYZ, B=CameraPositionWS())) {
   bUsePreviewValueAsDefault=true
   description=This will typically be world position - camera position.
   inputName=Camera Relative Position
   sortPriority=-1
  }
  n2 = Constant3Vector() {constant=<rgba 1 0 0 1>}
  n3 = FunctionInput(Preview=Transform(None=@n2) {transformSourceType=TRANSFORMSOURCE_Camera}) {bUsePreviewValueAsDefault=true, description="1,0,0 transformed from camera space to world space.", inputName=X Camera to World Vector}
  n4 = Multiply(A=@n2, B=DotProduct(A=@n1, B=@n3) {desc=local camera x}) {desc=x component}
  n5 = Constant3Vector() {constant=<rgba 0 1 0 1>}
  n6 = FunctionInput(Preview=Transform(None=@n5) {transformSourceType=TRANSFORMSOURCE_Camera}) {
   bUsePreviewValueAsDefault=true
   description="0,1,0 transformed from camera space to world space."
   inputName=Y Camera to World Vector
   sortPriority=1
  }
  n7 = Multiply(A=@n5, B=DotProduct(A=@n1, B=@n6)) {desc=y component}
  n8 = FunctionInput(Preview=Transform(None=Constant3Vector() {constant=<rgba 0 0 1 1>}) {transformSourceType=TRANSFORMSOURCE_Camera}) {
   bUsePreviewValueAsDefault=true
   description="0,0,1 transformed from camera space to world space."
   inputName=Z Camera to World Vector
   sortPriority=2
  }
  n9 = FunctionInput(Preview=ViewProperty() {property=MEVP_TanHalfFieldOfView}.Property) {
   bUsePreviewValueAsDefault=true
   description="Tan of FOV/2 for x. For y it's the ratio of the screen's y resolution to the x resolution multiplied by tan of fov/2. See an example within this function. "
   inputName="Tan(FOV/2)*[1,Screen Res Y/X]"
   inputType=FunctionInput_Vector2
   sortPriority=4
  }
  n10 = Divide(A=ComponentMask(None=Add(A=@n4, B=@n7)) {g=true, r=true}, B=Multiply(A=DotProduct(A=@n1, B=@n8), B=@n9))
  n11 = Constant2Vector() {g=-0.5, r=0.5}
  FunctionOutput(None=Add(A=Multiply(A=@n10, B=@n11) {constB=0.5}, B=Constant2Vector() {g=0.5, r=0.5}) {constB=0.5}) {outputName=Screen UVs}
  n12 = MaterialFunctionCall() {materialFunction=/Engine/Functions/Engine_MaterialFunctions02/Utility/FOV}
  n13 = MaterialFunctionCall(Float2=MaterialFunctionCall() {materialFunction=&ScreenResolution}.Visible Resolution) {materialFunction=&BreakOutFloat2Components}
  Multiply(A=Tangent!(None=Divide(A=@n12.FOV angle in degrees)) {period=360}, B=AppendVector(A=Constant() {r=1}, B=Divide(A=@n13.G, B=@n13.G)))
 ]
}
