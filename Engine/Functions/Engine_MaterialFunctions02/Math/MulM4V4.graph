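# MulM4V4 is a library-exposed material function that multiplies a 4x4 matrix by a four-component vector.
# The matrix arrives as four Vector4 row inputs, and the Vector input is shared through a named reroute.
# A dot product of each row with the vector gives one component, and MakeFloat4 packs the four results into the output.
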MaterialFunction /Engine/Functions/Engine_MaterialFunctions02/Math/MulM4V4
graph {
 properties {bExposeToLibrary=true}
 nodes [
  n1 = FunctionInput() {inputName=MatrixRow1, inputType=FunctionInput_Vector4}
  n2 = NamedRerouteUsage()
  n3 = FunctionInput() {inputName=MatrixRow2, inputType=FunctionInput_Vector4}
  n4 = NamedRerouteUsage()
  n5 = FunctionInput() {inputName=MatrixRow3, inputType=FunctionInput_Vector4}
  n6 = NamedRerouteUsage()
  n7 = FunctionInput() {inputName=MatrixRow4, inputType=FunctionInput_Vector4}
  n8 = NamedRerouteUsage()
  FunctionOutput(None=MaterialFunctionCall(X=DotProduct(A=@n1, B=@n2), Y=DotProduct(A=@n3, B=@n4), Z=DotProduct(A=@n5, B=@n6), A=DotProduct(A=@n7, B=@n8)) {materialFunction=&MakeFloat4}.Result)
  NamedRerouteDeclaration(None=FunctionInput() {inputName=Vector, inputType=FunctionInput_Vector4}) {name=Vector4, nodeColor=<rgba 1 0 0.0703 1>}
 ]
}
