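# MultiplyAdd is an engine material function exposed to the library under Misc and Math.
# It multiplies the Add input by Base, then adds Base to the product, giving Base + Add × Base.
# The result scales Base by one plus Add, so Add acts as a modulated boost rather than a plain offset.
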MaterialFunction /Engine/Functions/Engine_MaterialFunctions02/Math/MultiplyAdd
graph {
 properties {
  bExposeToLibrary=true
  description="Modulate the \"add\" input by by the base input and then add the result to the base input."
  libraryCategoriesText [Misc,Math]
 }
 nodes [
  n1 = FunctionInput() {inputName=Add, sortPriority=1}
  n2 = FunctionInput() {inputName=Base}
  FunctionOutput(None=Add(A=Multiply(A=@n1, B=@n2), B=@n2))
 ]
}
