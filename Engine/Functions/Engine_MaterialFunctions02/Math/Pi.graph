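# Material function Pi that returns pi multiplied by an input value, exposed to the material library under Misc and Math.
# A scalar Multiplier input, which uses its preview value of 1 as the default, is multiplied by a constant approximating pi and sent to the single output.
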MaterialFunction /Engine/Functions/Engine_MaterialFunctions02/Math/Pi
graph {
 properties {bExposeToLibrary=true, description=Pi multiplied by an input value., libraryCategoriesText [Misc,Math]}
 nodes [
  n1 = FunctionInput() {bUsePreviewValueAsDefault=true, description=Multiply pi, inputName=Multiplier, inputType=FunctionInput_Scalar, previewValue=<1 0 0 1>}
  FunctionOutput(None=Multiply(A=Constant() {r=3.1416}, B=@n1)) {description=Pi multiplied by the multiplier value.}
 ]
}
