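# ProjectVectorOntoPlane is an engine material function in the Math library category that projects a vector onto a plane.
# It takes Vector and Plane Normal inputs, dots them, scales the normal by that result, and subtracts it from the vector.
# The normal is never normalized inside the function, so the projection is exact only for a unit-length normal.
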MaterialFunction /Engine/Functions/Engine_MaterialFunctions02/Math/ProjectVectorOntoPlane
graph {
 properties {
  bExposeToLibrary=true
  description="Projects a vector onto a plane by removing it's deviation along the plane's normal. "
  libraryCategoriesText [Math]
 }
 nodes [
  n1 = FunctionInput() {inputName=Vector, previewValue=<1 1 1 1>}
  n2 = FunctionInput() {inputName=Plane Normal, previewValue=<0 0 1 1>}
  n3 = DotProduct(A=@n1, B=@n2)
  FunctionOutput(None=Subtract(A=@n1, B=Multiply(A=@n3, B=@n2)))
 ]
}
