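# Material function in the Misc and Math library categories that converts an RGB color to HSV, output in the 0-1 range.
# Channel max and min give value and delta, saturation is delta over max, and chained If nodes pick and wrap hue.
# Zero delta forces hue and saturation to zero.
# Debug value readouts and an HSV-to-RGB sector network on a constant input sit outside the output path.
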MaterialFunction /Engine/Functions/Engine_MaterialFunctions02/Math/RGBtoHSV
graph {
 properties {bExposeToLibrary=true, description="Converts RGB colors to HSV (hue saturation and value)", libraryCategoriesText [Misc,Math]}
 nodes [
  n1 = FunctionInput() {desc=0.019092+0.185148+0.08333, inputName=RGB, previewValue=<0.1 0 0.6749 1>}
  n2 = MaterialFunctionCall(Float3=@n1) {materialFunction=&BreakOutFloat3Components}
  n3 = Max(A=Max(A=@n2.R, B=@n2.R), B=@n2.B)
  n4 = Subtract(A=@n3, B=Min(A=Min(A=@n2.R, B=@n2.R), B=@n2.B))
  n5 = Constant()
  n6 = Constant()
  n7 = Constant() {r=0.6667}
  n8 = Divide(A=Subtract(A=@n3, B=@n1)) {constB=6}
  n9 = Divide(A=@n4)
  n10 = MaterialFunctionCall(Float3=Divide(A=Add(A=@n8, B=@n9), B=@n4)) {materialFunction=&BreakOutFloat3Components}
  n11 = If(A=@n2.B, B=@n3, "A > B"=@n6, "A == B"=Subtract(A=Add(A=@n7, B=@n10.G), B=@n10.R), "A < B"=@n6)
  n12 = Constant() {r=0.3333}
  n13 = If(A=@n2.G, B=@n3, "A > B"=@n11, "A == B"=Subtract(A=Add(A=@n12, B=@n10.R), B=@n10.B), "A < B"=@n11)
  n14 = If(A=@n3, B=@n2.R, "A > B"=@n13, "A == B"=Subtract(A=@n10.B, B=@n10.B), "A < B"=@n13)
  n15 = Constant()
  n16 = Add(A=@n14, B=If(A=@n14, B=@n15, "A > B"=@n15, "A == B"=@n15, "A < B"=Constant() {r=1}))
  n17 = Constant()
  n18 = Add(A=If(A=@n16, B=Constant() {r=1}, "A > B"=Constant() {r=-1}, "A == B"=@n17, "A < B"=@n17), B=@n16)
  n19 = If(A=@n4, B=@n5, "A > B"=@n18, "A == B"=@n5, "A < B"=@n18)
  n20 = Constant()
  n21 = Divide(A=@n4, B=@n3)
  n22 = If(A=@n4, B=@n20, "A > B"=@n21, "A == B"=@n20, "A < B"=@n21)
  n23 = Multiply(
   A=FunctionOutput(None=MaterialFunctionCall(X=@n19, Y=@n22, Z=@n3) {materialFunction=&MakeFloat3}.Result) {outputName="RGB to HSV (0-1)"}
   B=Constant3Vector() {constant=<rgba 360 100 100 0>}
  ) {constB=360}
  n24 = Multiply(A=@n1) {constB=255}
  n25 = TextureCoordinate()
  n26 = Constant2Vector() {g=0.045}
  n27 = Add(A=@n25, B=@n26)
  n28 = Multiply(A=@n26) {constB=2}
  Add(
   A=Add(
    A=MaterialFunctionCall!(Vector3=@n23, MaximumNumberOfDigits, UVs, DebugTextLocation RG_UpperRight BA_LowerLeft, Component Spacing) {materialFunction=/Engine/Functions/Engine_MaterialFunctions02/Utility/DebugFloat3Values}.ColorCodedOutput
    B=MaterialFunctionCall(Vector3=@n24, UVs=@n27) {materialFunction=&DebugFloat3Values}.ColorCodedOutput
   )
   B=MaterialFunctionCall(Vector3=@n1, UVs=Add(A=@n25, B=@n28)) {materialFunction=&DebugFloat3Values}.ColorCodedOutput
  )
  n29 = Constant3Vector() {desc=HSV}
  n30 = MaterialFunctionCall(Float3=@n29) {materialFunction=&BreakOutFloat3Components}
  n31 = Multiply(A=@n30.R) {constB=6}
  n32 = If(A=@n31, B=Constant() {r=6}, "A > B"=@n31, "A == B"=Constant(), "A < B"=@n31) {desc=h}
  n33 = Floor(None=@n32) {desc=i}
  n34 = OneMinus(None=@n30.G)
  n35 = Multiply(A=@n30.B, B=@n34) {desc=1}
  n36 = Subtract(A=@n32, B=@n33)
  n37 = Multiply(A=Multiply(A=@n36, B=@n34), B=@n30.B) {desc=2}
  n38 = MaterialFunctionCall(X=@n30.B, Y=@n35, Z=@n37) {materialFunction=&MakeFloat3}
  n39 = Multiply(A=@n30.B, B=Multiply(A=@n34, B=OneMinus(None=@n36))) {desc=3}
  n40 = If(
   A=@n33
   B=Constant() {r=4}
   "A > B"=@n38.Result
   "A == B"=MaterialFunctionCall(X=@n39, Y=@n35, Z=@n30.B) {materialFunction=&MakeFloat3}.Result
   "A < B"=@n38.Result
  )
  n41 = If(
   A=@n33
   B=Constant() {r=3}
   "A > B"=@n40
   "A == B"=MaterialFunctionCall(X=@n35, Y=@n37, Z=@n30.B) {materialFunction=&MakeFloat3}.Result
   "A < B"=@n40
  )
  n42 = If(
   A=@n33
   B=Constant() {r=2}
   "A > B"=@n41
   "A == B"=MaterialFunctionCall(X=@n35, Y=@n30.B, Z=@n39) {materialFunction=&MakeFloat3}.Result
   "A < B"=@n41
  )
  n43 = If(
   A=@n33
   B=Constant() {r=1}
   "A > B"=@n42
   "A == B"=MaterialFunctionCall(X=@n37, Y=@n30.B, Z=@n35) {materialFunction=&MakeFloat3}.Result
   "A < B"=@n42
  )
  n44 = If(
   A=@n33
   B=Constant()
   "A > B"=@n43
   "A == B"=MaterialFunctionCall(X=@n30.B, Y=@n39, Z=@n35) {materialFunction=&MakeFloat3}.Result
   "A < B"=@n43
  )
  MaterialFunctionCall(Vector3=If(A=@n30.G, B=Constant(), "A > B"=@n44, "A == B"=@n29, "A < B"=@n44) {desc="s == 0"}) {materialFunction=&DebugFloat3Values}
 ]
}
