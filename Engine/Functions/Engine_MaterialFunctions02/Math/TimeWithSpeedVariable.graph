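# TimeWithSpeedVariable is a library-exposed material function that scales time by a speed factor for animating materials.
# It multiplies a Time input, which falls back to the Time node, by a scalar Speed input that defaults to one.
# It outputs both the raw scaled time and its fractional part, which gives a repeating zero-to-one ramp.
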MaterialFunction /Engine/Functions/Engine_MaterialFunctions02/Math/TimeWithSpeedVariable
graph {
 properties {bExposeToLibrary=true, description=Time multiplied by speed with an option to use the remainder or time itself.}
 nodes [
  n1 = FunctionInput() {bUsePreviewValueAsDefault=true, inputName=Speed, inputType=FunctionInput_Scalar, previewValue=<1 0 0 1>}
  n2 = Multiply(A=FunctionInput(Preview=Time()) {bUsePreviewValueAsDefault=true, inputName=Time}, B=@n1)
  FunctionOutput(None=@n2) {outputName=Time}
  FunctionOutput(None=Frac(None=@n2)) {outputName=Frac Time}
 ]
}
