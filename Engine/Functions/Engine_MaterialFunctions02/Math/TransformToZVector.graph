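# Material function, exposed to the library, that transforms a vector into a basis built around a given Z vector, for example world position relative to hit normals.
# CreateThirdOrthogonalVector takes the normalized Z Vector and a constant +X reference, and its output feeds the basis inputs of InverseTransformMatrix.
# Center Location is subtracted from the input vector, which defaults to world position, before the transform, allowing centering around points like object position.
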MaterialFunction /Engine/Functions/Engine_MaterialFunctions02/Math/TransformToZVector
graph {
 properties {
  bExposeToLibrary=true
  description=This transforms a vector based on a Z vector and generated orthogonal X and Y vectors. Useful for transforming worldposition based on Hit normals etc.
 }
 nodes [
  n1 = FunctionInput(Preview=WorldPosition().XYZ) {bUsePreviewValueAsDefault=true, inputName=Vector to Transform, previewValue=<0 0 1 1>}
  n2 = FunctionInput() {
   bUsePreviewValueAsDefault=true
   description=Center Location will be subtracted from the Transformed Vector which is useful for centering the transformed vectors around something like ObjectPosition for example. Defaults to 0.
   inputName=Center Location
   sortPriority=2
  }
  n3 = MaterialFunctionCall!(
   Vector1=Normalize(VectorInput=FunctionInput() {bUsePreviewValueAsDefault=true, inputName=Z Vector, sortPriority=1, previewValue=<0 0 1 1>})
   Vector2=Constant3Vector() {constant=<rgba 1 0 0 0>}
  ) {materialFunction=&/CreateThirdOrthogonalVector}
  FunctionOutput(None=MaterialFunctionCall!(VectorToTransform=Subtract(A=@n1, B=@n2), BasisX=@n3.Vector2, BasisY=@n3.Vector2, BasisZ=@n3.Vector2) {materialFunction=&/InverseTransformMatrix}.Transformed Vector 3)
 ]
}
