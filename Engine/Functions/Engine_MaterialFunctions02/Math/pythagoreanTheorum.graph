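# Material function that derives the second leg of a right triangle from the first using the Pythagorean theorem, with the hypotenuse fixed at length one.
# It squares the scalar Side 1 Length input, subtracts that from one and takes the square root, outputting Side 2 Length.
# The graph does not clamp the value before the square root.
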MaterialFunction /Engine/Functions/Engine_MaterialFunctions02/Math/pythagoreanTheorum
graph {nodes [
  n1 = FunctionInput() {inputName=Side 1 Length, inputType=FunctionInput_Scalar}
  n2 = Multiply(A=@n1, B=@n1)
  FunctionOutput(None=SquareRoot(None=OneMinus(None=@n2))) {outputName=Side 2 Length}
 ]}
